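# UI sound cue for the frontend feat display, apparently played when completed feats are shown.
# A mixer layers an immediate map hover sound with a community vote increase sound that starts after a short fixed delay.
# An enveloper above the delay shapes the second layer with a volume curve that eases down, then drops to silence.
# It uses the Fort_UI sound class and overrides concurrency to allow two instances, stopping the oldest.
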SoundCue /Game/Sounds/Fort_UI_Sounds/Gameplay/Feats/Feats_Frontend_ShowCompleted_Cue
root=Mixer_0
nodes [
 {id=Delay_0, children [WavePlayer_1], properties {delayMax=0.4, delayMin=0.4}}
 {
  id=Enveloper_0
  children [Delay_0]
  properties {
   pitchCurve {editorCurveData {keys [{time=0}]}}
   volumeCurve {editorCurveData {keys [
      {interpMode=RCIM_Cubic, leaveTangent=-0.1619, tangentMode=RCTM_Break, time=0, value=1}
      {arriveTangent=-0.3513, interpMode=RCIM_Cubic, leaveTangent=-1, tangentMode=RCTM_Break, time=0.75, value=0.9}
      <at 1 0>
     ]}}
  }
 }
 {id=Mixer_0, children [WavePlayer_0,Enveloper_0], properties {inputVolume [1,1]}}
 {id=WavePlayer_0, properties {soundWaveAssetPtr=/Game/Sounds/Fort_UI_Sounds/FrontEnd/New_FrontEnd/Fortnite_UI_Map_Hover_01}}
 {id=WavePlayer_1, properties {soundWaveAssetPtr=/Game/Sounds/Fort_UI_Sounds/FrontEnd/CommunityVote/UI_CommunityVote_Increase_01}}
]
properties {
 bOverrideConcurrency=true
 maxDistance=2097152
 soundClassObject=&Fort_UI
 volumeMultiplier=1
 concurrencyOverrides {maxCount=2, resolutionRule=StopOldest}
}
